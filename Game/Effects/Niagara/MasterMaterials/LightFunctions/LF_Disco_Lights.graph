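# Light function material that projects moving disco lights through a soft window shape, apparently for a party bus.
# Two panning samples of a party-lights texture, blurred by a chosen mip level, add to a pulse driven by a MusicPulsing function output.
# A static switch can leave only the pulse.
# A beveled-rectangle mask, built from a sphere mask on exponentiated UVs, and a V-axis falloff shape the result.
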Material /Game/Effects/Niagara/MasterMaterials/LightFunctions/LF_Disco_Lights
graph {
 properties {
  bForceCompatibleWithLightFunctionAtlas=true
  materialDomain=MD_LightFunction
  pixelDepthOffsetMode=PDOM_Legacy
  usedShadingModels=Default Lit
 }
 nodes [
  n1 = NamedRerouteUsage()
  n2 = AppendVector(A=ScalarParameter() {defaultValue=5, parameterName=Lights_U_Scale_B}, B=ScalarParameter() {defaultValue=2, parameterName=Lights_V_Scale_B})
  n3 = ScalarParameter() {defaultValue=0.781, parameterName=Lights_U_PanSpeed_B}
  n4 = ScalarParameter() {defaultValue=0.0485, parameterName=Lights_V_PanSpeed_B}
  n5 = ScalarParameter() {defaultValue=4, parameterName=BlurredLightsMIPLevel}
  n6 = TextureSample!(UVs=Panner(Coordinate=Multiply(A=@n1, B=@n2), Speed=AppendVector(A=@n3, B=@n4)) {bFractionalPart=true}, Tex, Level=@n5, Apply View MipBias) {
   mipValueMode=TMVM_MipLevel
   samplerType=SAMPLERTYPE_LinearColor
   texture=/Game/Packages/Fortress_SharedMaterials/PartyBus_TempLocation/T_ArmoredBattleBus_PartyLights
  }
  n7 = NamedRerouteUsage()
  n8 = AppendVector(A=ScalarParameter() {defaultValue=4.5, parameterName=Lights_U_Scale}, B=ScalarParameter() {defaultValue=1.5, parameterName=Lights_V_Scale})
  n9 = ScalarParameter() {defaultValue=0.44, parameterName=Lights_U_PanSpeed}
  n10 = ScalarParameter() {defaultValue=-0.062, parameterName=Lights_V_PanSpeed}
  n11 = TextureSample(UVs=Panner(Coordinate=Multiply(A=@n7, B=@n8), Speed=AppendVector(A=@n9, B=@n10)) {bFractionalPart=true, speedX=0.22}, Level=@n5) {
   mipValueMode=TMVM_MipLevel
   samplerType=SAMPLERTYPE_LinearColor
   texture=&T_ArmoredBattleBus_PartyLights
  }
  n12 = LinearInterpolate(
   A=ScalarParameter() {defaultValue=0.18, parameterName=PulseLow}
   B=ScalarParameter() {defaultValue=1, parameterName=PulseHigh}
   Alpha=MaterialFunctionCall!(In) {materialFunction=/Game/Packages/Fortress_SharedMaterials/PartyBus_TempLocation/MF_ArmoredBattleBus_PartyLights}.MusicPulsing
  )
  n13 = Add(A=Add(A=@n6.R, B=@n11.R), B=@n12)
  n14 = Multiply(
   A=Subtract(A=TextureCoordinate()) {constB=0.5}
   B=AppendVector(
    A=ScalarParameter() {defaultValue=1.5, parameterName=WindowMaskScale_U}
    B=ScalarParameter() {defaultValue=8, parameterName=WindowMaskScale_V}
   )
  ) {constB=2}
  n15 = SphereMask(
   A=Power(Base=Abs(None=@n14), Exp=ScalarParameter() {defaultValue=5, parameterName=WindowMaskBeveledEdgeExponent}) {constExponent=1}
   B=Constant()
   Radius=Constant() {r=1}
   Hardness=ScalarParameter() {parameterName=WindowMaskHardness}
  )
  n16 = ScalarParameter() {defaultValue=50, parameterName=WindowMaskExponent}
  n17 = MaterialFunctionCall(Float2=TextureCoordinate()) {materialFunction=&BreakOutFloat2Components}
  n18 = Constant() {r=0.5}
  n19 = ScalarParameter() {defaultValue=0.4, parameterName=FinalGradientRadius}
  n20 = ScalarParameter() {defaultValue=0.02, parameterName=FinalGradientHardness}
  Multiply(
   A=Multiply(
    A=StaticSwitchParameter(True=@n13, False=@n12) {defaultValue=true, parameterName=AddDiscoLights}
    B=Power(Base=@n15, Exp=@n16) {constExponent=5}
   )
   B=Power(Base=SphereMask(A=@n17.G, B=@n18, Radius=@n19, Hardness=@n20), Exp=ScalarParameter() {defaultValue=8, parameterName=FinalGradientExponent}) {constExponent=5}
  )
  NamedRerouteDeclaration(None=TextureCoordinate()) {name=UVs, nodeColor=<rgba 0 0.5781 1 1>}
 ]
}
