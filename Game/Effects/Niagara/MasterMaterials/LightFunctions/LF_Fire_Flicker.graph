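# Light function material that modulates a light with animated grayscale noise for a fire flicker effect.
# A scrolling sample of a blurred noise texture is blended against white, using a mask that oscillates between static and panned samples of that same texture on a time-driven sine.
# A power parameter shapes the result, with further parameters for UV tiling and main scroll speed.
# The material is flagged as compatible with the light function atlas.
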Material /Game/Effects/Niagara/MasterMaterials/LightFunctions/LF_Fire_Flicker
graph {
 properties {
  bForceCompatibleWithLightFunctionAtlas=true
  materialDomain=MD_LightFunction
  pixelDepthOffsetMode=PDOM_Legacy
  usedShadingModels=Default Lit
 }
 nodes [
  n1 = TextureCoordinate()
  n2 = Constant2Vector() {g=1, r=1}
  n3 = MaterialFunctionCall(R=ScalarParameter() {defaultValue=1, parameterName=UV Multiplier X}, G=ScalarParameter() {defaultValue=1, parameterName=UV Multiplier Y}) {materialFunction=&AppendMany}
  n4 = Panner(
   Coordinate=Multiply(A=@n1, B=Multiply(A=@n2, B=@n3.RG))
   Time=Multiply(A=Time(), B=ScalarParameter() {defaultValue=0.4, parameterName=Time Scale})
  ) {speedY=-1}
  n5 = Multiply(A=@n1, B=Multiply(A=Constant2Vector() {g=1, r=2}, B=@n3.RG))
  n6 = TextureSample(UVs=@n5) {samplerType=SAMPLERTYPE_Grayscale, texture=&LowResBlurredNoise}
  n7 = Constant() {r=0.5}
  n8 = TextureSample(UVs=Panner(Coordinate=Add(A=@n5, B=@n7)) {speedY=0.25}) {samplerType=SAMPLERTYPE_Grayscale, texture=&LowResBlurredNoise}
  n9 = MaterialFunctionCall!(Sine Phase=Time(), Value 1, Value 2) {materialFunction=/Engine/Functions/Engine_MaterialFunctions02/Utility/Sine_Remapped}
  Power(
   Base=LinearInterpolate(
    B=TextureSample(UVs=@n4) {samplerType=SAMPLERTYPE_Grayscale, texture=&LowResBlurredNoise}.RGB
    Alpha=LinearInterpolate(A=@n6.RGB, B=@n8.RGB, Alpha=@n9.Remapped Sine)
   ) {constA=1, constB=0}
   Exp=ScalarParameter() {defaultValue=1, parameterName=Texture 1 power}
  )
 ]
}
