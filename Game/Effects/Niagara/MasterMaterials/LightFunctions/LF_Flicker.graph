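# Light function material that produces a stepped, noise-driven brightness flicker, set to be compatible with the light function atlas.
# Time scaled by a Frequency parameter wraps and quantizes into 32 steps that pan across a 32x32 nearest-neighbor noise texture.
# UVs shrink to about one texel, so each step samples roughly a single noise value.
# The sample is desaturated and clamped to a Minimum Value parameter, limiting how dark the flicker gets.
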Material /Game/Effects/Niagara/MasterMaterials/LightFunctions/LF_Flicker
graph {
 properties {
  bForceCompatibleWithLightFunctionAtlas=true
  materialDomain=MD_LightFunction
  pixelDepthOffsetMode=PDOM_Legacy
  usedShadingModels=Default Lit
 }
 nodes [
  n1 = Time()
  n2 = ScalarParameter() {defaultValue=1.4, parameterName=Frequency}
  n3 = Constant() {r=32}
  n4 = Multiply(A=Frac(None=Multiply(A=@n1, B=@n2)), B=@n3)
  n5 = Panner(Coordinate=Multiply(A=TextureCoordinate()) {constB=0.0312}, Time=Divide(A=Floor(None=@n4), B=@n3)) {speedX=1}
  Clamp(
   None=Desaturation(None=TextureSample(UVs=@n5) {samplerType=SAMPLERTYPE_LinearColor, texture=/Game/Packages/Fortress_Sky/Materials/Noise32x322_nearestNeighboor}.RGB)
   Min=ScalarParameter() {defaultValue=0.45, parameterName=Minimum Value}
  )
  LinearInterpolate() {constA=1, constB=0}
 ]
}
