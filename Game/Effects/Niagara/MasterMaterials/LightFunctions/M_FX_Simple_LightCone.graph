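# Additive, unlit, two-sided Niagara master material for light cone effects, tinted and faded by particle color.
# Two scrolling samples of a packed light pattern pass through looping four-mask blends, all fed the red channel, then crossfade on a sine wave.
# Brightness and power shape the pattern, which is multiplied by a double-sided Fresnel term blended by an invertible camera-to-object-axis falloff.
# Emissive passes through a shared adjustment function, while opacity uses depth fade except on mobile.
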Material /Game/Effects/Niagara/MasterMaterials/LightFunctions/M_FX_Simple_LightCone
graph {
 properties {
  bAutomaticallySetUsageInEditor=false
  bUsedWithNiagaraMeshParticles=true
  bUsedWithNiagaraRibbons=true
  bUsedWithNiagaraSprites=true
  blendMode=BLEND_Additive
  shadingModel=MSM_Unlit
  twoSided=true
  usedShadingModels=Unlit
 }
 nodes [
  Constant3Vector() {constant=<rgba 0 0 1 0>}
  n1 = ParticleColor()
  n2 = TextureCoordinate()
  n3 = Time()
  n4 = AppendVector(A=ScalarParameter() {defaultValue=-0.01, parameterName=Alt_Scroll_U}, B=ScalarParameter() {parameterName=Alt_Scroll_V})
  n5 = ScalarParameter() {defaultValue=3, parameterName=Alt_Tile_U}
  n6 = ScalarParameter() {defaultValue=1, parameterName=Alt_Tile_V}
  n7 = TextureSampleParameter2D(UVs=Multiply(A=Panner(Coordinate=@n2, Time=@n3, Speed=@n4), B=AppendVector(A=@n5, B=@n6))) {
   parameterName=Diffuse
   samplerType=SAMPLERTYPE_LinearColor
   texture=/Game/Effects/Niagara/MasterMaterials/DiffusePacked/Textures/T_FX_Packed_LightPatterns_01
  }
  n8 = Modulo!(A=Multiply(A=Time() {period=4}, B=ScalarParameter() {defaultValue=1, parameterName=Mask Cycle Speed}), B=Constant() {r=4})
  n9 = MaterialFunctionCall!(Mask 1=@n7.R, Mask 2=@n7.R, Mask 3=@n7.R, Mask 4=@n7.R, Blend=@n8) {materialFunction=/Game/Effects/Niagara/MasterMaterials/Functions/Utility/MF_Lerp_4_Scalar_Looping}
  n10 = TextureCoordinate()
  n11 = Time()
  n12 = AppendVector(A=ScalarParameter() {defaultValue=0.01, parameterName=Scroll_U}, B=ScalarParameter() {parameterName=Scroll_V})
  n13 = ScalarParameter() {defaultValue=3, parameterName=Tile_U}
  n14 = ScalarParameter() {defaultValue=1, parameterName=Tile_V}
  n15 = TextureSampleParameter2D(UVs=Multiply(A=Panner(Coordinate=@n10, Time=@n11, Speed=@n12), B=AppendVector(A=@n13, B=@n14))) {
   parameterName=Diffuse
   samplerType=SAMPLERTYPE_LinearColor
   texture=&T_FX_Packed_LightPatterns_01
  }
  n16 = MaterialFunctionCall(Mask 1=@n15.R, Mask 2=@n15.R, Mask 3=@n15.R, Mask 4=@n15.R, Blend=@n8) {materialFunction=&MF_Lerp_4_Scalar_Looping}
  n17 = Time()
  n18 = ScalarParameter() {defaultValue=1, parameterName=Alt Sample Cycle Speed}
  n19 = MaterialFunctionCall(Input=Sine(None=Multiply(A=@n17, B=@n18)), Input Low=Constant() {r=-1}) {materialFunction=&RemapValueRange}
  n20 = Power(
   Base=Multiply(A=LinearInterpolate(A=@n9.Result, B=@n16.Result, Alpha=@n19.Result), B=ScalarParameter() {defaultValue=1, parameterName=Brightness})
   Exp=ScalarParameter() {defaultValue=2, parameterName=Power}
  )
  n21 = Abs(None=CameraVectorWS())
  n22 = ObjectOrientation()
  n23 = Clamp(None=Power(Base=DotProduct(A=@n21, B=@n22), Exp=ScalarParameter() {defaultValue=5, parameterName=Camera Falloff}) {constExponent=5})
  n24 = LinearInterpolate(
   A=MaterialFunctionCall(
    Brightness=ScalarParameter() {defaultValue=3, parameterName=Fresnel Brightness}
    Exponent=ScalarParameter() {defaultValue=3, parameterName=Fresnel Falloff}
    Camera Vector=CameraVectorWS()
   ) {materialFunction=&MF_Fresnel_FX}.Fresnel Double Sided
   Alpha=StaticSwitchParameter(True=OneMinus(None=@n23), False=@n23) {parameterName=Invert Camera Fade}
  )
  MaterialFunctionCall(Emissive=Multiply(A=Multiply(A=@n1.RGB, B=@n20), B=@n24)) {materialFunction=&MF_FX_EmissiveAdjustments}
  n25 = Multiply(A=Multiply(A=@n1.A, B=@n20), B=@n24)
  n26 = DepthFade(Opacity=@n25, FadeDistance=ScalarParameter() {parameterName=DepthFade_FadeDistance})
  ShadingPathSwitch(Default=@n26, Deferred=@n26, Forward=@n26, Mobile=@n25)
 ]
}
